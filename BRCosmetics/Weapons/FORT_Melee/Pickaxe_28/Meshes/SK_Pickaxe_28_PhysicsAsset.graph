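# Physics asset for the skeletal mesh of Pickaxe_28, a two-handed melee cosmetic weapon in BRCosmetics.
# It holds a single body on the Melee_TwoHanded_Handle bone, using one rotated, offset box shape as simplified collision for the pickaxe.
# There are no constraints, so the asset defines no jointed or multi-body simulation.
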PhysicsAsset /BRCosmetics/Weapons/FORT_Melee/Pickaxe_28/Meshes/SK_Pickaxe_28_PhysicsAsset
bodies {Melee_TwoHanded_Handle [{shapeType=Box, extentX=76.6412, extentY=21.9285, extentZ=155.2538, center=<-4.9639 2.154 24.6171>, rotation=<rot -2.5083 -90 -3.8128>}]}
constraints []
